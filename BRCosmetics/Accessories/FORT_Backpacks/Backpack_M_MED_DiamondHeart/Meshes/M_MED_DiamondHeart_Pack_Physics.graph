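# Physics asset for the DiamondHeart backpack cosmetic, defining collision and joint limits for its dangling chain.
# A box and capsule cover the pack, while capsules and a sphere cover the chain bones.
# The main and two start bodies constrain to the pack, with left and right links joined to main.
# The center link ties main to both starts, forming closed constraint loops with mostly tight angular limits away from the pack.
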PhysicsAsset /BRCosmetics/Accessories/FORT_Backpacks/Backpack_M_MED_DiamondHeart/Meshes/M_MED_DiamondHeart_Pack_Physics
bodies {
 dyn_chain_C [{shapeType=Capsule, length=3.5, radius=2.5, shapeName=dyn_chain_C_capsule, center=<0.7759 -0.6656 0>}]
 dyn_chain_L [{shapeType=Capsule, length=4, radius=2, shapeName=dyn_chain_L_capsule}]
 dyn_chain_R [{shapeType=Capsule, length=4, radius=2, shapeName=dyn_chain_R_capsule}]
 dyn_chain_main [
  {shapeType=Sphere, radius=3, shapeName=dyn_chain_main_sphere, center=<6.766 -0.0699 0>}
  {shapeType=Capsule, length=1, radius=3.005, shapeName=dyn_chain_main_capsule, center=<1.4063 0.7582 0>, rotation=<rot 0 -90 0>}
 ]
 dyn_chain_startL [{shapeType=Capsule, length=2.2702, radius=2, shapeName=dyn_chain_startL_capsule, center=<-2.281 0.5384 0.5408>}]
 dyn_chain_startR [{shapeType=Capsule, length=2.2702, radius=2, shapeName=dyn_chain_startR_capsule, center=<2.281 -0.5419 -0.5408>}]
 pack [
  {shapeType=Capsule, length=5.5, radius=3.5, shapeName=pack_capsule, center=<0 -2.4417 -10.7689>, rotation=<rot 0 0 -10>}
  {shapeType=Box, extentX=32.5, extentY=17.5, extentZ=41.5, shapeName=pack_box, center=<0 10.1583 -14.5921>, rotation=<rot 0 0 2.8125>}
 ]
}
constraints [
 {bone1Name=dyn_chain_C, bone2Name=dyn_chain_main, swing1LimitDegrees=1, swing2LimitDegrees=10, twistLimitDegrees=1}
 {bone1Name=dyn_chain_R, bone2Name=dyn_chain_main, swing1LimitDegrees=10, swing2LimitDegrees=10, twistLimitDegrees=10}
 {bone1Name=dyn_chain_L, bone2Name=dyn_chain_main, swing1LimitDegrees=10, swing2LimitDegrees=10, twistLimitDegrees=10}
 {bone1Name=dyn_chain_main, bone2Name=pack, swing1LimitDegrees=90, swing2LimitDegrees=60, twistLimitDegrees=60}
 {bone1Name=dyn_chain_startL, bone2Name=pack, swing1LimitDegrees=29.16, swing2LimitDegrees=43.56, twistLimitDegrees=45}
 {bone1Name=dyn_chain_startR, bone2Name=pack, swing1LimitDegrees=29.16, swing2LimitDegrees=43.56, twistLimitDegrees=45}
 {bone1Name=dyn_chain_C, bone2Name=dyn_chain_startR, swing1LimitDegrees=50, swing2LimitDegrees=10, twistLimitDegrees=10}
 {bone1Name=dyn_chain_C, bone2Name=dyn_chain_startL, swing1LimitDegrees=50, swing2LimitDegrees=10, twistLimitDegrees=10}
]
