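# Animation Blueprint for the DiamondHeart backpack cosmetic, adding physics-driven secondary motion.
# A RigidBody simulation feeds spring-based AnimDynamics on both gun bones, with a tighter, stiffer pair replacing the looser one in the front end.
# Bone modifiers then rotate the main chain during skydiving and offset the root by owner spine bend.
# Each update sets gravity from skydive or sprint state and physics alpha from pawn speed.
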AnimBlueprint /BRCosmetics/Accessories/FORT_Backpacks/Backpack_M_MED_DiamondHeart/Meshes/M_MED_DiamondHeart_Pack_AnimBP
target_skeleton=&/M_MED_DiamondHeart_Pack_Skeleton
graphs [
 AnimGraph {nodes [
   n1 = Get OffOnFromtEnd() -> (OffOnFromtEnd:float)
   n2 = Anim Dynamics - Bone: dyn_L_gun(Alpha=@n1, ExternalForce="0,0,0") {
    angularDampingOverride=1
    angularSpringConstant=10000
    bAngularSpring=true
    bOverrideAngularDamping=true
    bUseGravityOverride=true
    gravityScale=0.35
    simulationSpace=RootRelative
    alphaBoolBlend {blendOption=Linear}
    boundBone {boneName=dyn_L_gun}
    componentLinearAccScale=<1 1 1>
    componentLinearVelScale=<1 1 1>
    gravityOverride {z=-980}
    physicsBodyDefinitions [{
      boundBone {boneName=dyn_L_gun}
      boxExtents=<60 60 60>
      constraintSetup {angularLimitsMax=<5 0 5>, angularLimitsMin=<-5 -10 -5>, angularTarget=<1 0 0>}
      localJointOffset=<4 0 0>
     }]
   }
   Anim Dynamics - Bone: dyn_R_gun(ComponentPose=@n2, Alpha=@n1, ExternalForce="0,0,0") {
    angularDampingOverride=1
    angularSpringConstant=10000
    bAngularSpring=true
    bOverrideAngularDamping=true
    bUseGravityOverride=true
    gravityScale=0.35
    simulationSpace=RootRelative
    alphaBoolBlend {blendOption=Linear}
    boundBone {boneName=dyn_R_gun}
    componentLinearAccScale=<1 1 1>
    componentLinearVelScale=<1 1 1>
    gravityOverride {z=-980}
    physicsBodyDefinitions [{
      boundBone {boneName=dyn_R_gun}
      boxExtents=<55 55 55>
      constraintSetup {angularLimitsMax=<5 0 5>, angularLimitsMin=<-5 -10 -5>, angularTarget=<1 0 0>}
      localJointOffset=<-4 0 0>
     }]
   }
   n3 = RigidBody!(
    ComponentPose=Local To Component(LocalPose=Input Pose())
    Alpha=Get Alpha() -> (Alpha:float)
    OverrideWorldGravity:Vector=Get GravityOverride()
    ExternalForce
    ComponentLinearAccScale:Vector=Get Component Linear Acc Scale() -> (Component Linear Acc Scale:Vector)
   ) -> (Pose) {
    bClampLinearTranslationLimitToRefPose=true
    bOverrideWorldGravity=true
    overridePhysicsAsset=&/M_MED_DiamondHeart_Pack_Physics
    alphaBoolBlend {blendOption=Linear}
    baseBoneRef {boneName=root}
    componentAppliedLinearAccClamp=<2000 2000 1000>
    componentLinearAccScale=<1 1 0.3>
    componentLinearVelScale=<1.2 1 1>
    overrideWorldGravity {z=-980}
    simSpaceSettings {
     maxAngularAcceleration=20
     maxAngularVelocity=200
     maxLinearAcceleration=100
     maxLinearVelocity=10
     velocityScaleZ=0.6
     worldAlpha=0.5
     externalLinearDragV=<0.1 0.1>
    }
   }
   n4 = Anim Dynamics - Bone: dyn_L_gun(ComponentPose=@n3, Alpha=@n1, ExternalForce="0,0,0") {
    angularDampingOverride=1
    angularSpringConstant=10000
    bAngularSpring=true
    bOverrideAngularDamping=true
    bUseGravityOverride=true
    gravityScale=0.35
    simulationSpace=RootRelative
    alphaBoolBlend {blendOption=Linear}
    boundBone {boneName=dyn_L_gun}
    componentLinearAccScale=<1 1 1>
    componentLinearVelScale=<1 1 1>
    gravityOverride {z=-980}
    physicsBodyDefinitions [{
      boundBone {boneName=dyn_L_gun}
      boxExtents=<60 60 60>
      constraintSetup {angularLimitsMax=<3 0 3>, angularLimitsMin=<-3 -6 -3>, angularTarget=<1 0 0>}
      localJointOffset=<1 0 0>
     }]
   }
   n5 = Anim Dynamics - Bone: dyn_R_gun(ComponentPose=@n4, Alpha=@n1, ExternalForce="0,0,0") {
    angularDampingOverride=1
    angularSpringConstant=10000
    bAngularSpring=true
    bOverrideAngularDamping=true
    bUseGravityOverride=true
    gravityScale=0.35
    simulationSpace=RootRelative
    alphaBoolBlend {blendOption=Linear}
    boundBone {boneName=dyn_R_gun}
    componentLinearAccScale=<1 1 1>
    componentLinearVelScale=<1 1 1>
    gravityOverride {z=-980}
    physicsBodyDefinitions [{
      boundBone {boneName=dyn_R_gun}
      boxExtents=<55 55 55>
      constraintSetup {angularLimitsMax=<3 0 3>, angularLimitsMin=<-3 -6 -3>, angularTarget=<1 0 0>}
      localJointOffset=<-1 0 0>
     }]
   }
   n6 = Get OnOnFrontEnd() -> (OnOnFrontEnd:float)
   n7 = Anim Dynamics - Bone: dyn_L_gun(ComponentPose=@n5, Alpha=@n6, ExternalForce="0,0,0") {
    angularDampingOverride=1
    angularSpringConstant=15000
    bAngularSpring=true
    bOverrideAngularDamping=true
    bUseGravityOverride=true
    gravityScale=0.35
    simulationSpace=RootRelative
    alphaBoolBlend {blendOption=Linear}
    boundBone {boneName=dyn_L_gun}
    componentLinearAccScale=<1 1 1>
    componentLinearVelScale=<1 1 1>
    gravityOverride {z=-980}
    physicsBodyDefinitions [{
      boundBone {boneName=dyn_L_gun}
      boxExtents=<60 60 60>
      constraintSetup {angularLimitsMax=<1.5 0 1.5>, angularLimitsMin=<-1.5 -3 -1.5>, angularTarget=<1 0 0>}
      localJointOffset=<0.4 0 0>
     }]
   }
   n8 = Anim Dynamics - Bone: dyn_R_gun(ComponentPose=@n7, Alpha=@n6, ExternalForce="0,0,0") {
    angularDampingOverride=1
    angularSpringConstant=15000
    bAngularSpring=true
    bOverrideAngularDamping=true
    bUseGravityOverride=true
    gravityScale=0.35
    simulationSpace=RootRelative
    alphaBoolBlend {blendOption=Linear}
    boundBone {boneName=dyn_R_gun}
    componentLinearAccScale=<1 1 1>
    componentLinearVelScale=<1 1 1>
    gravityOverride {z=-980}
    physicsBodyDefinitions [{
      boundBone {boneName=dyn_R_gun}
      boxExtents=<55 55 55>
      constraintSetup {angularLimitsMax=<1.5 0 1.5>, angularLimitsMin=<-1.5 -3 -1.5>, angularTarget=<1 0 0>}
      localJointOffset=<-0.4 0 0>
     }]
   }
   n9 = "Transform (Modify) Bone - Bone: dyn_chain_main"!(ComponentPose=@n8, bAlphaBoolEnabled:bool=Get bIsSkyDiving(), Translation="0,0,0", Rotation="-0.000,0.000,20.000", Scale="1,1,1") -> (Pose) {
    alphaInputType=Bool
    rotationMode=BMM_Additive
    rotationSpace=BCS_ParentBoneSpace
    alphaBoolBlend {blendOption=Linear}
    boneToModify {boneName=dyn_chain_main}
    rotation {roll=20}
   }
   n10 = "Transform (Modify) Bone - Bone: root"(ComponentPose=@n9, Alpha=Get back_rotate_alpha() -> (back_rotate_alpha:double), Translation="0,-2.25,0", Rotation="0,0,15") {
    rotationMode=BMM_Additive
    rotationSpace=BCS_BoneSpace
    translationMode=BMM_Additive
    translationSpace=BCS_BoneSpace
    alphaBoolBlend {blendOption=Linear}
    boneToModify {boneName=root}
    rotation {roll=15}
    translation {y=-2.25}
   }
   Output Pose(Result=Component To Local(ComponentPose=@n10)) {layerGroup=None}
  ]}
 EventGraph {nodes [
   Event BlueprintUpdateAnimation()
   > n1 = Sequence!(execute) -> (then_0, then_1, then_2:Exec, then_3:Exec)
   n2 = BreakTransform(InTransform=GetDeltaTransformFromRefPose(self=GetOwnerSkeletalMeshComponent(), BoneName=spine_05, BaseName=spine_04))
   > Set back_rotate_alpha(back_rotate_alpha:double=MapRangeClamped(Value=BreakRotator(InRot=@n2.Rotation).Yaw, InRangeA=0, InRangeB=-25)) -> (Output_Get:double)
   > n3 = Branch(Condition=Get bIsSkyDiving())
   > Set GravityOverride(GravityOverride="0,0,50")
   n4 = Branch(execute=@n3.else, Condition=Get bIsSprinting())
   > Set GravityOverride(GravityOverride="0,0,-700")
   Set GravityOverride(execute=@n4.else, GravityOverride="0,0,-400")
   n5 = Get PawnVelocity()
   n6 = Branch(execute=@n1.then_1, Condition="float < float"(A=Vector Length(A=@n5), B=1))
   > Set Alpha(Alpha:float=0.6) -> (Output_Get:float)
   Set Alpha(execute=@n6.else, Alpha=1)
   n7 = Branch(execute=@n1.then_2, Condition=Get bIsInFrontEnd())
   > Set OffOnFromtEnd(OffOnFromtEnd:float=0) -> (Output_Get:float)
   > Set OnOnFrontEnd(OnOnFrontEnd:float=1) -> (Output_Get:float)
   Set OffOnFromtEnd(execute=@n7.else, OffOnFromtEnd=1)
   > Set OnOnFrontEnd(OnOnFrontEnd=0)
  ]}
]
variables [
 Alpha {type=number, default=1}
 back_rotate_alpha {type=number}
 Component Linear Acc Scale {type=Vector, default=<1 1 0.3>}
 GravityOverride {type=Vector}
 OffOnFromtEnd {type=number, default=1}
 OnOnFrontEnd {type=number}
]
